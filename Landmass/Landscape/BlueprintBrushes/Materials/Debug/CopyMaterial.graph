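# CopyMaterial is an unlit debug material in the Landmass landscape Blueprint brush content that produces packed height data from a texture.
# It multiplies the red channel of the engine's CloudWeatherTexture by a scalar Cm parameter to form a height value.
# The PackRG8Height material function takes that value on its signed 32k-range integer height input to pack it.
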Material /Landmass/Landscape/BlueprintBrushes/Materials/Debug/CopyMaterial
graph {
 properties {pixelDepthOffsetMode=PDOM_Legacy, shadingModel=MSM_Unlit, usedShadingModels=Unlit}
 nodes [MaterialFunctionCall!(Height Int -32k to 32k=Multiply(
    A=TextureSample() {texture=/Engine/EngineSky/VolumetricClouds/CloudWeatherTexture}.R
    B=ScalarParameter() {defaultValue=5000, parameterName=Cm}
   )) {materialFunction=/Landmass/Landscape/OldPrototype_BP/Mats/MF/PackRG8Height}]
}
