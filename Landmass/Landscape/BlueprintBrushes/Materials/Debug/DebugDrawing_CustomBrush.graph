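# Unlit, alpha-composite debug material with depth test disabled that visualizes Landmass custom brush drawing on landscapes.
# A grid scaled by the MPC_Landscape render-target resolution is tinted red using scene-depth comparison and DepthFade.
# It adds a green cursor mask and P0-P1 line from MPC_Drawing, sized by render-target world size.
# Four offset samples of the Shape texture are subtracted from the center to draw green edge outlines.
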Material /Landmass/Landscape/BlueprintBrushes/Materials/Debug/DebugDrawing_CustomBrush
graph {
 properties {bDisableDepthTest=true, blendMode=BLEND_AlphaComposite, pixelDepthOffsetMode=PDOM_Legacy, shadingModel=MSM_Unlit, usedShadingModels=Unlit}
 nodes [
  n1 = TextureCoordinate()
  n2 = CollectionParameter() {collection=/Landmass/Landscape/BlueprintBrushes/MPC/MPC_Landscape, parameterName=RTResX}
  n3 = CollectionParameter() {collection=&MPC_Landscape, parameterName=RTResY}
  n4 = Divide(A=Subtract(A=AppendVector(A=@n2, B=@n3))) {constB=64, desc=Draw grid every N Quads}
  n5 = TextureSample(UVs=Multiply(A=Multiply(A=@n1, B=@n4)) {constB=4}) {texture=/Engine/EngineDebugMaterials/T_1x1_Grid}
  n6 = Add(A=@n5.RGB) {constB=0.1}
  n7 = Constant3Vector() {constant=<rgba 0.5 0 0 0>}
  n8 = Subtract(A=PixelDepth().R, B=SceneDepth().R)
  n9 = ScalarParameter() {defaultValue=-15000, parameterName=DF2}
  n10 = Saturate(None=Divide(A=Add(A=@n8, B=@n9), B=@n9))
  n11 = ScalarParameter() {defaultValue=0.3, parameterName=Min Red'}
  n12 = LinearInterpolate(A=LinearInterpolate(B=Multiply(A=@n6, B=@n7), Alpha=Max(A=@n10, B=@n11)), B=@n5.RGB, Alpha=DepthFade())
  n13 = ScalarParameter() {defaultValue=0.5, parameterName=Grid}
  n14 = CollectionParameter() {collection=/Landmass/Landscape/OldPrototype_BP/Mats/MPC/MPC_Drawing, parameterName=Cursor}
  n15 = MaterialFunctionCall(Float2=Divide(A=CollectionParameter() {collection=&MPC_Landscape, parameterName=RTWorldSize}) {constB=100}) {materialFunction=&BreakOutFloat2Components}
  n16 = Max(A=@n15.R, B=@n15.R)
  n17 = Add(
   A=SphereMask(A=WorldPosition().XYZ, B=ComponentMask(None=@n14) {b=true, g=true, r=true}, Radius=@n16) {hardnessPercent=0}
   B=MaterialFunctionCall!(
    Coordinates=WorldPosition().XYZ
    P0=CollectionParameter() {collection=&MPC_Drawing, parameterName=P0}
    P1=CollectionParameter() {collection=&MPC_Drawing, parameterName=P1}
    Line Width=@n16
    Edge Falloff
    Square Corners?
   ) {materialFunction=/Engine/Functions/Engine_MaterialFunctions01/Coordinates/DrawLine-2D}.Result
  )
  n18 = Constant3Vector() {constant=<rgba 0 1 0 0>}
  n19 = ScalarParameter() {defaultValue=1, parameterName=ShapeAlpha}
  n20 = TextureCoordinate()
  n21 = TextureSampleParameter2D(UVs=@n20) {parameterName=Shape, texture=&Black}
  n22 = Divide(B=AppendVector(
    A=CollectionParameter() {collection=&MPC_Landscape, parameterName=RTResX}
    B=CollectionParameter() {collection=&MPC_Landscape, parameterName=RTResY}
   )) {constA=4}
  n23 = Multiply(A=@n22, B=Constant2Vector() {r=1})
  n24 = TextureCoordinate()
  n25 = Multiply(A=@n22, B=Constant2Vector() {g=1})
  n26 = MaterialFunctionCall(
   X=TextureSampleParameter2D(UVs=Add(A=@n20, B=@n23)) {parameterName=Shape, texture=&Black}.R
   Y=TextureSampleParameter2D(UVs=Subtract(A=@n20, B=@n23)) {parameterName=Shape, texture=&Black}.R
   Z=TextureSampleParameter2D(UVs=Add(A=@n24, B=@n25)) {parameterName=Shape, texture=&Black}.R
   A=TextureSampleParameter2D(UVs=Subtract(A=@n24, B=@n25)) {parameterName=Shape, texture=&Black}.R
  ) {materialFunction=&MakeFloat4}
  n27 = Multiply(A=DotProduct(A=Subtract(A=@n21.R, B=@n26.Result), B=Constant() {r=1}), B=Constant3Vector() {constant=<rgba 0 0.5 0 0>})
  Add(A=Add(A=Multiply(A=@n12, B=@n13) {constB=0.2}, B=Multiply(A=@n17, B=@n18)), B=Saturate(None=Multiply(A=@n19, B=@n27)))
  LinearInterpolate()
  MaterialFunctionCall!(Texture Object, UVs, Contrast, Highpass Offset, High Quality) {materialFunction=/Engine/Functions/Engine_MaterialFunctions01/ImageAdjustment/HighPassTexture}
  Multiply(A=OneMinus(None=DepthFade()), B=ScalarParameter() {defaultValue=0.75, parameterName=Clipping Min Display})
  ScalarParameter() {parameterName=Radius}
 ]
}
