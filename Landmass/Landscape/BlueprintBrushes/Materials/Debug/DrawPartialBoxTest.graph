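# Unlit debug material in the Landmass Blueprint brushes, apparently for testing partial box drawing, that produces a flat encoded landscape height.
# A scalar parameter Z feeds the PackRG8Height function as a signed integer height, packing it into two 8-bit channels.
# The material has no textures, masks or shape logic of its own, and it allows negative emissive color.
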Material /Landmass/Landscape/BlueprintBrushes/Materials/Debug/DrawPartialBoxTest
graph {
 properties {bAllowNegativeEmissiveColor=true, pixelDepthOffsetMode=PDOM_Legacy, shadingModel=MSM_Unlit, usedShadingModels=Unlit}
 nodes [MaterialFunctionCall!(Height Int -32k to 32k=ScalarParameter() {defaultValue=10000, parameterName=Z}) {materialFunction=/Landmass/Landscape/BlueprintBrushes/MF/PackRG8Height}]
}
