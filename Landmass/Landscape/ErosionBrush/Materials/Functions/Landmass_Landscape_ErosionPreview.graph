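# Material function that previews Landmass erosion on a landscape, adjusting normals and world position offset from render-target data.
# HeightRT, optionally RG8-unpacked, supplies terrain height for vertical offset, gated by a static Enable switch and EnableLandmassPreview parameter.
# Terrain normals blend between NormalRT and custom HLSL that bilinearly filters central-difference height normals.
# Incoming material normals are reoriented via TransformToZVector; outputs are modified attributes plus separate Normal and WPO.
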MaterialFunction /Landmass/Landscape/ErosionBrush/Materials/Functions/Landmass_Landscape_ErosionPreview
graph {nodes [
  FunctionOutput(None=LinearInterpolate(A=VertexNormalWS(), B=NamedRerouteUsage(), Alpha=NamedRerouteUsage())) {outputName=Normal, sortPriority=1}
  FunctionOutput(None=Multiply(A=NamedRerouteUsage(), B=NamedRerouteUsage())) {outputName=WPO, sortPriority=2}
  n1 = GetMaterialAttributes(None=FunctionInput() {inputName=Attributes, inputType=FunctionInput_MaterialAttributes}) {attributeGetTypes [#1, #2]}
  n2 = Normalize(VectorInput=Multiply(A=NamedRerouteUsage(), B=Constant3Vector() {constant=<rgba -1 -1 1 0>}))
  n3 = Transform(None=Normalize(VectorInput=MaterialFunctionCall!(Vector to Transform=@n1.Normal, Z Vector=@n2, Center Location) {materialFunction=/Engine/Functions/Engine_MaterialFunctions02/Math/TransformToZVector}.Result)) {transformSourceType=TRANSFORMSOURCE_World, transformType=TRANSFORM_Tangent}
  n4 = NamedRerouteUsage()
  n5 = Multiply(A=NamedRerouteUsage(), B=NamedRerouteUsage())
  FunctionOutput(None=SetMaterialAttributes!(
    MaterialAttributes=@n1.MaterialAttributes
    Normal=LinearInterpolate(A=@n1.Normal, B=@n3, Alpha=@n4)
    WorldPositionOffset=Add(A=@n5, B=@n1.WorldPositionOffset)
    BaseColor
   ) {attributeSetTypes [#1, #2, #3]}) {outputName=Attributes}
  n6 = TextureCoordinate()
  n7 = ScalarParameter() {defaultValue=0.5, parameterName=TexelCenter}
  n8 = CollectionParameter() {collection=/Landmass/Landscape/BlueprintBrushes/MPC/MPC_Landscape, parameterName=RTResX}
  n9 = CollectionParameter() {collection=&MPC_Landscape, parameterName=RTResY}
  NamedRerouteDeclaration(None=Divide(A=Add(A=@n6, B=@n7), B=AppendVector(A=@n8, B=@n9))) {name=Landscape UVs, nodeColor=<rgba 1 0 0.2109 1>}
  n10 = Multiply(A=WorldPosition().Z) {constB=-1}
  n11 = NamedRerouteUsage()
  NamedRerouteDeclaration(None=Multiply(A=Constant3Vector() {constant=<rgba 0 0 1 0>}, B=Add(A=@n10, B=@n11))) {name=WPO, nodeColor=<rgba 0.9453 1 0 1>}
  n12 = TextureSampleParameter2D(UVs=NamedRerouteUsage()) {parameterName=HeightRT, samplerSource=SSM_Wrap_WorldGroupSettings, texture=&Black}
  n13 = MaterialFunctionCall!(RG=@n12.RGB) {materialFunction=/Landmass/Landscape/BlueprintBrushes/MF/UnpackRG8Height}
  n14 = ScalarParameter() {parameterName=RG8Height}
  NamedRerouteDeclaration(None=Multiply(
    A=LinearInterpolate(A=@n12.R, B=@n13.Result, Alpha=@n14)
    B=CollectionParameter() {collection=&MPC_Landscape, parameterName=LandscapeZScale}
   )) {name=Landscape Z, nodeColor=<rgba 0 1 0.0156 1>}
  n15 = StaticBool() {value=true}
  NamedRerouteDeclaration(None=StaticSwitch(
    True=ScalarParameter() {parameterName=EnableLandmassPreview}
    False=Constant()
    Value=FunctionInput(Preview=@n15) {bUsePreviewValueAsDefault=true, inputName=Enable, inputType=FunctionInput_StaticBool, sortPriority=2}
   )) {name=Erosion Preview, nodeColor=<rgba 1 0 0.75 1>}
  MaterialExpressionTextureObjectParameter_1 = TextureObjectParameter() {parameterName=HeightRT, texture=&Black}
  MaterialExpressionReroute_2 = Reroute(None=AppendVector(
    A=CollectionParameter() {collection=&MPC_Landscape, parameterName=RTResX}
    B=CollectionParameter() {collection=&MPC_Landscape, parameterName=RTResY}
   ))
  MaterialExpressionNamedRerouteUsage_12 = NamedRerouteUsage()
  MaterialExpressionReroute_1 = Reroute(None=ComponentMask(None=CollectionParameter() {collection=&MPC_Landscape, parameterName=LandscapeScale}) {r=true})
  MaterialExpressionCollectionParameter_4 = CollectionParameter() {collection=&MPC_Landscape, parameterName=LandscapeZScale}
  n16 = Custom!(
   HeightTex=@MaterialExpressionTextureObjectParameter_1
   TextureSize=@MaterialExpressionReroute_2
   UV=@MaterialExpressionNamedRerouteUsage_12
   GridSize=@MaterialExpressionReroute_1
   HeightScale=@MaterialExpressionCollectionParameter_4
  ) {
   code="//HeightTex: input Texture Object containing heights.R and water.B\r\n//TextureSize: input for size\r\n//GridSize: input float grid world size\r\n//HeightScale: input for height range scale\r\n\r\n\r\nstruct NormalSampleHelper\r\n{\r\n\r\n\tfloat2 Offsets[4];\r\n\tfloat2 TexelSize;\r\n\tfloat HeightScaleValue;\r\n\tfloat GridSizeValue;\r\n\r\n\tfloat UnpackRG8(float2 inSourceSample)\r\n\t{\r\n\t\treturn clamp(float(((int)round(inSourceSample.r * 255.0) << 8) | (int)round(inSourceSample.g * 255.0)), 0.0, 65536.0);\r\n\t}\r\n\t\r\n\tfloat3 SampleNormal(float2 inUV, Texture2D inTexture, SamplerState inTextureSampler)\r\n\t{\r\n\t\tfloat2 NearestUV = (floor(inUV / TexelSize)*TexelSize) + (TexelSize * 0.5);\r\n\r\n\t\tfloat2 PosYSample = UnpackRG8(inTexture.SampleLevel(inTextureSampler, NearestUV + Offsets[0] * TexelSize, 0).rg);\r\n\t\tfloat2 NegYSample = UnpackRG8(inTexture.SampleLevel(inTextureSampler, NearestUV + Offsets[1] * TexelSize, 0).rg);\r\n\t\tfloat2 PosXSample = UnpackRG8(inTexture.SampleLevel(inTextureSampler, NearestUV + Offsets[2] * TexelSize, 0).rg);\r\n\t\tfloat2 NegXSample = UnpackRG8(inTexture.SampleLevel(inTextureSampler, NearestUV + Offsets[3] * TexelSize, 0).rg);\r\n\t\t\r\n\t\tfloat3 Tangent =  float3( 2*GridSizeValue, 0, ((PosXSample.x) - (NegXSample.x))*HeightScaleValue);\r\n\t\tfloat3 Binormal = float3( 0, 2*GridSizeValue, ((PosYSample.x) - (NegYSample.x))*HeightScaleValue);\r\n\t\tTangent = normalize(Tangent);\r\n\t\tBinormal = normalize(Binormal);\r\n\t\t\r\n\t\treturn normalize(cross(Tangent, Binormal));\r\n\t}\r\n\t\r\n\tfloat3 SampleNormal_BilinearFilter(float2 inUV, Texture2D inTexture, SamplerState inTextureSampler)\r\n\t{\r\n\t\tinUV -= (TexelSize * 0.5);\r\n\t\r\n\t\tfloat2 NearestUV = (floor(inUV / TexelSize)*TexelSize) + (TexelSize * 0.5);\r\n\r\n\t\tfloat3 Sample_x0y0 = SampleNormal(NearestUV + float2(0,0) * TexelSize, inTexture, inTextureSampler);\r\n\t\tfloat3 Sample_x1y0 = SampleNormal(NearestUV + float2(1,0) * TexelSize, inTexture, inTextureSampler);\r\n\t\tfloat3 Sample_x0y1 = SampleNormal(NearestUV + float2(0,1) * TexelSize, inTexture, inTextureSampler);\r\n\t\tfloat3 Sample_x1y1 = SampleNormal(NearestUV + float2(1,1) * TexelSize, inTexture, inTextureSampler);\r\n\t\t\r\n\t\tfloat2 BlendPhase = frac( inUV / TexelSize );\r\n\t\t\r\n\t\tfloat3 Blend_X0 = lerp(Sample_x0y0, Sample_x1y0, BlendPhase.x);\r\n\t\tfloat3 Blend_X1 = lerp(Sample_x0y1, Sample_x1y1, BlendPhase.x);\r\n\t\tfloat3 OutNormal = lerp(Blend_X0, Blend_X1, BlendPhase.y);\r\n\t\t\r\n\t\treturn normalize(OutNormal);\r\n\t}\r\n\t\r\n};\r\n\r\nNormalSampleHelper Helper;\r\nHelper.Offsets[0] = float2(0,1);\r\nHelper.Offsets[1] = float2(0,-1); \r\nHelper.Offsets[2] = float2(1,0); \r\nHelper.Offsets[3] = float2(-1,0);\r\nHelper.TexelSize = 1 / TextureSize;\r\nHelper.HeightScaleValue = HeightScale;\r\nHelper.GridSizeValue = GridSize;\r\n\r\n\r\nreturn Helper.SampleNormal_BilinearFilter(UV, HeightTex, HeightTexSampler);\r\n//return Helper.SampleNormal(UV, HeightTex, HeightTexSampler);\r\n"
   description=Compute Normal
   inputs [
    {
     inputName=HeightTex
     input {expression=&self:MaterialExpressionTextureObjectParameter_1}
    }
    {
     inputName=TextureSize
     input {expression=&self:MaterialExpressionReroute_2}
    }
    {
     inputName=UV
     input {expression=&self:MaterialExpressionNamedRerouteUsage_12}
    }
    {
     inputName=GridSize
     input {expression=&self:MaterialExpressionReroute_1}
    }
    {
     inputName=HeightScale
     input {expression=&self:MaterialExpressionCollectionParameter_4}
    }
   ]
  }
  NamedRerouteDeclaration(None=@n16) {name=DerivedNormal, nodeColor=<rgba 0 0.625 1 1>}
  n17 = NamedRerouteUsage()
  n18 = ConstantBiasScale(None=ComponentMask(None=TextureSampleParameter2D(UVs=@n17) {parameterName=NormalRT, samplerSource=SSM_Wrap_WorldGroupSettings, texture=&Black}.RGB) {g=true, r=true}) {bias=-0.5, scale=2}
  NamedRerouteDeclaration(None=LinearInterpolate(A=DeriveNormalZ(InXY=@n18), B=NamedRerouteUsage(), Alpha=ScalarParameter() {parameterName=UseDerivedNormal})) {name=Terrain Normal, nodeColor=<rgba 0.6875 1 0 1>}
 ]}
