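# Debug landscape material for the Landmass Blueprint Brushes, giving a basic surface with an erosion preview.
# A Mud landscape layer sample blends between a scalar Base value and a MudColor parameter.
# The Landmass_Landscape_ErosionPreview function, taken from the ErosionBrush materials, is called with its Enable switch fixed on.
# The material uses Default Lit shading with tangent-space normals turned off.
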Material /Landmass/Landscape/BlueprintBrushes/Materials/Debug/Landscape_Basic
graph {
 properties {bTangentSpaceNormal=false, pixelDepthOffsetMode=PDOM_Legacy, usedShadingModels=Default Lit}
 nodes [
  LinearInterpolate(
   A=ScalarParameter() {defaultValue=0.185, parameterName=Base}
   B=VectorParameter() {parameterName=MudColor, defaultValue=<rgba 0.375 0.2333 0.0176 1>}.RGB
   Alpha=LandscapeLayerSample() {parameterName=Mud}
  )
  MaterialFunctionCall!(Attributes, Enable=StaticBool() {value=true}) {materialFunction=/Landmass/Landscape/ErosionBrush/Materials/Functions/Landmass_Landscape_ErosionPreview}
 ]
}
