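# Debug Landmass landscape brush material that shapes a spherical dome into the height render target.
# It maps UVs to landscape world space via MPC_Landscape, deriving a Pythagorean dome profile and edge alpha from distance to point P.
# A rotated, scaled sample of the editor Ai_Spawnpoint texture adds height.
# The result blends with unpacked HeightRT height through lerp, min, max or additive modes, gains DebugOffset, and is repacked as RG8.
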Material /Landmass/Landscape/BlueprintBrushes/Materials/Debug/M_Sphere
graph {
 properties {pixelDepthOffsetMode=PDOM_Legacy, shadingModel=MSM_Unlit, usedShadingModels=Unlit}
 nodes [
  ComponentMask() {g=true, r=true}
  Divide(A=ScalarParameter() {defaultValue=4096, parameterName=Size})
  n1 = LinearInterpolate(A=NamedRerouteUsage(), B=NamedRerouteUsage(), Alpha=NamedRerouteUsage())
  n2 = Min(A=NamedRerouteUsage(), B=NamedRerouteUsage())
  n3 = Max(A=NamedRerouteUsage(), B=NamedRerouteUsage())
  n4 = Add(A=NamedRerouteUsage(), B=NamedRerouteUsage())
  MaterialFunctionCall!(Height Int -32k to 32k=Add(
    A=MaterialFunctionCall!(AlphaBlend=@n1, Min=@n2, Max=@n3, Additive=@n4) {materialFunction=/Landmass/Landscape/OldPrototype_BP/Mats/MF/BrushBlendModes_Manual}.Result
    B=ScalarParameter() {parameterName=DebugOffset}
   )) {materialFunction=/Landmass/Landscape/BlueprintBrushes/MF/PackRG8Height}
  n5 = TextureSampleParameter2D(UVs=MaterialFunctionCall() {materialFunction=/Landmass/Landscape/BlueprintBrushes/MF/PointSampleLandscapeUVs}.TexelCenterUVs) {parameterName=HeightRT, texture=/Engine/EngineResources/Black_Low.Black}
  NamedRerouteDeclaration(None=MaterialFunctionCall!(RG=ComponentMask(None=@n5.RGB) {g=true, r=true}) {materialFunction=/Landmass/Landscape/OldPrototype_BP/Mats/MF/UnpackRG8Height}.Result) {name=Previous Height, nodeColor=<rgba 0 0.4609 1 1>}
  n6 = Multiply(
   A=TextureCoordinate()
   B=ComponentMask(None=CollectionParameter() {collection=/Landmass/Landscape/BlueprintBrushes/MPC/MPC_Landscape, parameterName=RTWorldSize}) {g=true, r=true}
  )
  n7 = Multiply(A=ComponentMask(None=CollectionParameter() {collection=&MPC_Landscape, parameterName=LandscapeScale}) {g=true, r=true}) {constB=0.5}
  n8 = CollectionParameter() {collection=&MPC_Landscape, parameterName=LandscapeLocation}
  n9 = Subtract(
   A=Add(A=Subtract(A=@n6, B=@n7), B=ComponentMask(None=@n8) {g=true, r=true})
   B=ComponentMask(None=VectorParameter() {parameterName=P}.RGB) {g=true, r=true}
  )
  n10 = VectorParameter() {parameterName=BrushRotation, defaultValue=<rgba 1 0 0 0>}
  n11 = ComponentMask(None=@n10.RGB) {g=true, r=true}
  n12 = MaterialFunctionCall(X=Multiply(A=@n10.G) {constB=-1}, Y=@n10.R) {materialFunction=&MakeFloat2}
  n13 = VectorParameter() {parameterName=BrushScale, defaultValue=<rgba 1 1 1 0>}
  n14 = ScalarParameter() {defaultValue=4096, parameterName=Size}
  n15 = Divide(
   A=AppendVector(A=DotProduct(A=@n9, B=@n11), B=DotProduct(A=@n9, B=@n12.Result))
   B=ComponentMask(None=Multiply(A=@n13.RGB, B=@n14)) {g=true, r=true}
  )
  n16 = TextureSample(UVs=Saturate(None=Add(A=@n15) {constB=0.5})) {texture=/Engine/EditorResources/Ai_Spawnpoint}
  n17 = ScalarParameter() {parameterName=TextureHeight}
  n18 = Multiply(
   A=TextureCoordinate()
   B=ComponentMask(None=CollectionParameter() {collection=&MPC_Landscape, parameterName=RTWorldSize}) {g=true, r=true}
  )
  n19 = Multiply(A=ComponentMask(None=CollectionParameter() {collection=&MPC_Landscape, parameterName=LandscapeScale}) {g=true, r=true}) {constB=0.5}
  n20 = CollectionParameter() {collection=&MPC_Landscape, parameterName=LandscapeLocation}
  n21 = Subtract(
   A=Add(A=Subtract(A=@n18, B=@n19), B=ComponentMask(None=@n20) {g=true, r=true})
   B=ComponentMask(None=VectorParameter() {parameterName=P}.RGB) {g=true, r=true}
  )
  n22 = ScalarParameter() {defaultValue=4096, parameterName=Size}
  n23 = Saturate(None=Divide(A=Length(None=@n21) {is=class MaterialExpressionLength}, B=Divide(A=@n22)))
  n24 = MaterialFunctionCall!(Side 1 Length=@n23) {desc=SafeClip, materialFunction=/Engine/Functions/Engine_MaterialFunctions02/Math/pythagoreanTheorum}
  n25 = ScalarParameter() {parameterName=Height}
  NamedRerouteDeclaration(None=Add(A=Multiply(A=@n16.R, B=@n17), B=Multiply(A=@n24.Side 2 Length, B=@n25))) {name=Current Height, nodeColor=<rgba 0.1484 1 0 1>}
  n26 = OneMinus(None=@n23)
  n27 = ScalarParameter() {defaultValue=0.2, parameterName=Alpha Width}
  NamedRerouteDeclaration(None=Saturate(None=Divide(A=@n26, B=@n27))) {name=Brush Alpha, nodeColor=<rgba 1 0 0.1641 1>}
  TextureCoordinate()
 ]
}
