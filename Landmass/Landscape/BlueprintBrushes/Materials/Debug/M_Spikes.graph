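# Debug unlit landscape-brush material containing two separate height calculations, with no material output connection shown.
# One maps UVs to world XY through MPC_Landscape, turns distance from point P into a dome of radius Size via the Pythagorean function, and packs it with PackRG8Height.
# The other marks every Nth eight-texel cell and scales it by H.
# Its Min reads R twice, so the pattern varies only along U, forming stripes.
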Material /Landmass/Landscape/BlueprintBrushes/Materials/Debug/M_Spikes
graph {
 properties {pixelDepthOffsetMode=PDOM_Legacy, shadingModel=MSM_Unlit, usedShadingModels=Unlit}
 nodes [
  n1 = TextureCoordinate()
  n2 = ComponentMask(None=CollectionParameter() {collection=/Landmass/Landscape/BlueprintBrushes/MPC/MPC_Landscape, parameterName=RTWorldSize}) {g=true, r=true}
  n3 = CollectionParameter() {collection=&MPC_Landscape, parameterName=LandscapeLocation}
  n4 = Subtract(
   A=Add(A=Multiply(A=@n1, B=@n2), B=ComponentMask(None=@n3) {g=true, r=true})
   B=ComponentMask(None=VectorParameter() {parameterName=P}.RGB) {g=true, r=true}
  )
  n5 = Saturate(None=Divide(A=Length(None=@n4) {is=class MaterialExpressionLength}, B=ScalarParameter() {defaultValue=4096, parameterName=Size}))
  MaterialFunctionCall!(Height Int -32k to 32k=Multiply(
    A=MaterialFunctionCall!(Side 1 Length=@n5) {desc=SafeClip, materialFunction=/Engine/Functions/Engine_MaterialFunctions02/Math/pythagoreanTheorum}.Side 2 Length
    B=ScalarParameter() {defaultValue=4096, parameterName=Size}
   )) {materialFunction=/Landmass/Landscape/BlueprintBrushes/MF/PackRG8Height}
  n6 = Multiply(A=TextureCoordinate(), B=CollectionParameter() {collection=&MPC_Landscape, parameterName=RTResX})
  n7 = Fmod(A=Floor(None=Divide(A=@n6) {constB=8}), B=ScalarParameter() {defaultValue=10, parameterName=Show Every N Particles})
  n8 = MaterialFunctionCall(Float2=OneMinus(None=Saturate(None=@n7))) {materialFunction=&BreakOutFloat2Components}
  Multiply(A=Saturate(None=Min(A=@n8.R, B=@n8.R)), B=ScalarParameter() {defaultValue=1024, parameterName=H})
 ]
}
