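# Unlit debug material in the Landmass Blueprint brushes that tests reading a data array from vector parameters by UV position.
# Parameters p0–p63 each pass through an Append into a Custom HLSL node, which indexes a 64-float array by UV.x.
# The array lists only p0–p15, and the computed 8x8 cell is unused.
# The result feeds DebugScalarValues, with no connection to a material output shown.
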Material /Landmass/Landscape/BlueprintBrushes/Materials/Debug/ReadTextureData
graph {
 properties {pixelDepthOffsetMode=PDOM_Legacy, shadingModel=MSM_Unlit, usedShadingModels=Unlit}
 nodes [
  MaterialExpressionTextureCoordinate_0 = TextureCoordinate()
  n1 = VectorParameter() {parameterName=p0, defaultValue=<rgba 0 0.333 0.666 1>}
  MaterialExpressionAppendVector_0 = AppendVector(A=@n1.RGB, B=@n1.RGB)
  n2 = VectorParameter() {parameterName=p1, defaultValue=<rgba 1 0.66 0.33 0>}
  MaterialExpressionAppendVector_1 = AppendVector(A=@n2.RGB, B=@n2.RGB)
  n3 = VectorParameter() {parameterName=p2, defaultValue=<rgba 1 0.66 0.33 0>}
  MaterialExpressionAppendVector_2 = AppendVector(A=@n3.RGB, B=@n3.RGB)
  n4 = VectorParameter() {parameterName=p3, defaultValue=<rgba 0 0.333 0.666 1>}
  MaterialExpressionAppendVector_3 = AppendVector(A=@n4.RGB, B=@n4.RGB)
  n5 = VectorParameter() {parameterName=p4, defaultValue=<rgba 0 0.333 0.666 1>}
  MaterialExpressionAppendVector_4 = AppendVector(A=@n5.RGB, B=@n5.RGB)
  n6 = VectorParameter() {parameterName=p5, defaultValue=<rgba 1 0.66 0.33 0>}
  MaterialExpressionAppendVector_5 = AppendVector(A=@n6.RGB, B=@n6.RGB)
  n7 = VectorParameter() {parameterName=p6, defaultValue=<rgba 1 0.66 0.33 0>}
  MaterialExpressionAppendVector_6 = AppendVector(A=@n7.RGB, B=@n7.RGB)
  n8 = VectorParameter() {parameterName=p7, defaultValue=<rgba 0 0.333 0.666 1>}
  MaterialExpressionAppendVector_7 = AppendVector(A=@n8.RGB, B=@n8.RGB)
  n9 = VectorParameter() {parameterName=p8, defaultValue=<rgba 0 0.333 0.666 1>}
  MaterialExpressionAppendVector_8 = AppendVector(A=@n9.RGB, B=@n9.RGB)
  n10 = VectorParameter() {parameterName=p9, defaultValue=<rgba 1 0.66 0.33 0>}
  MaterialExpressionAppendVector_9 = AppendVector(A=@n10.RGB, B=@n10.RGB)
  n11 = VectorParameter() {parameterName=p10, defaultValue=<rgba 1 0.66 0.33 0>}
  MaterialExpressionAppendVector_10 = AppendVector(A=@n11.RGB, B=@n11.RGB)
  n12 = VectorParameter() {parameterName=p11, defaultValue=<rgba 0 0.333 0.666 1>}
  MaterialExpressionAppendVector_11 = AppendVector(A=@n12.RGB, B=@n12.RGB)
  n13 = VectorParameter() {parameterName=p12, defaultValue=<rgba 0 0.333 0.666 1>}
  MaterialExpressionAppendVector_12 = AppendVector(A=@n13.RGB, B=@n13.RGB)
  n14 = VectorParameter() {parameterName=p13, defaultValue=<rgba 1 0.66 0.33 0>}
  MaterialExpressionAppendVector_13 = AppendVector(A=@n14.RGB, B=@n14.RGB)
  n15 = VectorParameter() {parameterName=p14, defaultValue=<rgba 1 0.66 0.33 0>}
  MaterialExpressionAppendVector_14 = AppendVector(A=@n15.RGB, B=@n15.RGB)
  n16 = VectorParameter() {parameterName=p15, defaultValue=<rgba 0 0.333 0.666 1>}
  MaterialExpressionAppendVector_15 = AppendVector(A=@n16.RGB, B=@n16.RGB)
  n17 = VectorParameter() {parameterName=p16, defaultValue=<rgba 0 0.333 0.666 1>}
  MaterialExpressionAppendVector_16 = AppendVector(A=@n17.RGB, B=@n17.RGB)
  n18 = VectorParameter() {parameterName=p17, defaultValue=<rgba 1 0.66 0.33 0>}
  MaterialExpressionAppendVector_17 = AppendVector(A=@n18.RGB, B=@n18.RGB)
  n19 = VectorParameter() {parameterName=p18, defaultValue=<rgba 1 0.66 0.33 0>}
  MaterialExpressionAppendVector_18 = AppendVector(A=@n19.RGB, B=@n19.RGB)
  n20 = VectorParameter() {parameterName=p19, defaultValue=<rgba 0 0.333 0.666 1>}
  MaterialExpressionAppendVector_19 = AppendVector(A=@n20.RGB, B=@n20.RGB)
  n21 = VectorParameter() {parameterName=p20, defaultValue=<rgba 0 0.333 0.666 1>}
  MaterialExpressionAppendVector_20 = AppendVector(A=@n21.RGB, B=@n21.RGB)
  n22 = VectorParameter() {parameterName=p21, defaultValue=<rgba 1 0.66 0.33 0>}
  MaterialExpressionAppendVector_21 = AppendVector(A=@n22.RGB, B=@n22.RGB)
  n23 = VectorParameter() {parameterName=p22, defaultValue=<rgba 1 0.66 0.33 0>}
  MaterialExpressionAppendVector_22 = AppendVector(A=@n23.RGB, B=@n23.RGB)
  n24 = VectorParameter() {parameterName=p23, defaultValue=<rgba 0 0.333 0.666 1>}
  MaterialExpressionAppendVector_23 = AppendVector(A=@n24.RGB, B=@n24.RGB)
  n25 = VectorParameter() {parameterName=p24, defaultValue=<rgba 0 0.333 0.666 1>}
  MaterialExpressionAppendVector_24 = AppendVector(A=@n25.RGB, B=@n25.RGB)
  n26 = VectorParameter() {parameterName=p25, defaultValue=<rgba 1 0.66 0.33 0>}
  MaterialExpressionAppendVector_25 = AppendVector(A=@n26.RGB, B=@n26.RGB)
  n27 = VectorParameter() {parameterName=p26, defaultValue=<rgba 1 0.66 0.33 0>}
  MaterialExpressionAppendVector_26 = AppendVector(A=@n27.RGB, B=@n27.RGB)
  n28 = VectorParameter() {parameterName=p27, defaultValue=<rgba 0 0.333 0.666 1>}
  MaterialExpressionAppendVector_27 = AppendVector(A=@n28.RGB, B=@n28.RGB)
  n29 = VectorParameter() {parameterName=p28, defaultValue=<rgba 0 0.333 0.666 1>}
  MaterialExpressionAppendVector_28 = AppendVector(A=@n29.RGB, B=@n29.RGB)
  n30 = VectorParameter() {parameterName=p29, defaultValue=<rgba 1 0.66 0.33 0>}
  MaterialExpressionAppendVector_29 = AppendVector(A=@n30.RGB, B=@n30.RGB)
  n31 = VectorParameter() {parameterName=p30, defaultValue=<rgba 1 0.66 0.33 0>}
  MaterialExpressionAppendVector_30 = AppendVector(A=@n31.RGB, B=@n31.RGB)
  n32 = VectorParameter() {parameterName=p31, defaultValue=<rgba 0 0.333 0.666 1>}
  MaterialExpressionAppendVector_31 = AppendVector(A=@n32.RGB, B=@n32.RGB)
  n33 = VectorParameter() {parameterName=p32, defaultValue=<rgba 0 0.333 0.666 1>}
  MaterialExpressionAppendVector_32 = AppendVector(A=@n33.RGB, B=@n33.RGB)
  n34 = VectorParameter() {parameterName=p33, defaultValue=<rgba 1 0.66 0.33 0>}
  MaterialExpressionAppendVector_33 = AppendVector(A=@n34.RGB, B=@n34.RGB)
  n35 = VectorParameter() {parameterName=p34, defaultValue=<rgba 1 0.66 0.33 0>}
  MaterialExpressionAppendVector_34 = AppendVector(A=@n35.RGB, B=@n35.RGB)
  n36 = VectorParameter() {parameterName=p35, defaultValue=<rgba 0 0.333 0.666 1>}
  MaterialExpressionAppendVector_35 = AppendVector(A=@n36.RGB, B=@n36.RGB)
  n37 = VectorParameter() {parameterName=p36, defaultValue=<rgba 0 0.333 0.666 1>}
  MaterialExpressionAppendVector_36 = AppendVector(A=@n37.RGB, B=@n37.RGB)
  n38 = VectorParameter() {parameterName=p37, defaultValue=<rgba 1 0.66 0.33 0>}
  MaterialExpressionAppendVector_37 = AppendVector(A=@n38.RGB, B=@n38.RGB)
  n39 = VectorParameter() {parameterName=p38, defaultValue=<rgba 1 0.66 0.33 0>}
  MaterialExpressionAppendVector_38 = AppendVector(A=@n39.RGB, B=@n39.RGB)
  n40 = VectorParameter() {parameterName=p39, defaultValue=<rgba 0 0.333 0.666 1>}
  MaterialExpressionAppendVector_39 = AppendVector(A=@n40.RGB, B=@n40.RGB)
  n41 = VectorParameter() {parameterName=p40, defaultValue=<rgba 0 0.333 0.666 1>}
  MaterialExpressionAppendVector_40 = AppendVector(A=@n41.RGB, B=@n41.RGB)
  n42 = VectorParameter() {parameterName=p41, defaultValue=<rgba 1 0.66 0.33 0>}
  MaterialExpressionAppendVector_41 = AppendVector(A=@n42.RGB, B=@n42.RGB)
  n43 = VectorParameter() {parameterName=p42, defaultValue=<rgba 1 0.66 0.33 0>}
  MaterialExpressionAppendVector_42 = AppendVector(A=@n43.RGB, B=@n43.RGB)
  n44 = VectorParameter() {parameterName=p43, defaultValue=<rgba 0 0.333 0.666 1>}
  MaterialExpressionAppendVector_43 = AppendVector(A=@n44.RGB, B=@n44.RGB)
  n45 = VectorParameter() {parameterName=p44, defaultValue=<rgba 0 0.333 0.666 1>}
  MaterialExpressionAppendVector_44 = AppendVector(A=@n45.RGB, B=@n45.RGB)
  n46 = VectorParameter() {parameterName=p45, defaultValue=<rgba 1 0.66 0.33 0>}
  MaterialExpressionAppendVector_45 = AppendVector(A=@n46.RGB, B=@n46.RGB)
  n47 = VectorParameter() {parameterName=p46, defaultValue=<rgba 1 0.66 0.33 0>}
  MaterialExpressionAppendVector_46 = AppendVector(A=@n47.RGB, B=@n47.RGB)
  n48 = VectorParameter() {parameterName=p47, defaultValue=<rgba 0 0.333 0.666 1>}
  MaterialExpressionAppendVector_47 = AppendVector(A=@n48.RGB, B=@n48.RGB)
  n49 = VectorParameter() {parameterName=p48, defaultValue=<rgba 0 0.333 0.666 1>}
  MaterialExpressionAppendVector_48 = AppendVector(A=@n49.RGB, B=@n49.RGB)
  n50 = VectorParameter() {parameterName=p49, defaultValue=<rgba 1 0.66 0.33 0>}
  MaterialExpressionAppendVector_49 = AppendVector(A=@n50.RGB, B=@n50.RGB)
  n51 = VectorParameter() {parameterName=p50, defaultValue=<rgba 1 0.66 0.33 0>}
  MaterialExpressionAppendVector_50 = AppendVector(A=@n51.RGB, B=@n51.RGB)
  n52 = VectorParameter() {parameterName=p51, defaultValue=<rgba 0 0.333 0.666 1>}
  MaterialExpressionAppendVector_51 = AppendVector(A=@n52.RGB, B=@n52.RGB)
  n53 = VectorParameter() {parameterName=p52, defaultValue=<rgba 0 0.333 0.666 1>}
  MaterialExpressionAppendVector_52 = AppendVector(A=@n53.RGB, B=@n53.RGB)
  n54 = VectorParameter() {parameterName=p53, defaultValue=<rgba 1 0.66 0.33 0>}
  MaterialExpressionAppendVector_53 = AppendVector(A=@n54.RGB, B=@n54.RGB)
  n55 = VectorParameter() {parameterName=p54, defaultValue=<rgba 1 0.66 0.33 0>}
  MaterialExpressionAppendVector_54 = AppendVector(A=@n55.RGB, B=@n55.RGB)
  n56 = VectorParameter() {parameterName=p55, defaultValue=<rgba 0 0.333 0.666 1>}
  MaterialExpressionAppendVector_55 = AppendVector(A=@n56.RGB, B=@n56.RGB)
  n57 = VectorParameter() {parameterName=p56, defaultValue=<rgba 0 0.333 0.666 1>}
  MaterialExpressionAppendVector_56 = AppendVector(A=@n57.RGB, B=@n57.RGB)
  n58 = VectorParameter() {parameterName=p57, defaultValue=<rgba 1 0.66 0.33 0>}
  MaterialExpressionAppendVector_57 = AppendVector(A=@n58.RGB, B=@n58.RGB)
  n59 = VectorParameter() {parameterName=p58, defaultValue=<rgba 1 0.66 0.33 0>}
  MaterialExpressionAppendVector_58 = AppendVector(A=@n59.RGB, B=@n59.RGB)
  n60 = VectorParameter() {parameterName=p59, defaultValue=<rgba 0 0.333 0.666 1>}
  MaterialExpressionAppendVector_59 = AppendVector(A=@n60.RGB, B=@n60.RGB)
  n61 = VectorParameter() {parameterName=p60, defaultValue=<rgba 0 0.333 0.666 1>}
  MaterialExpressionAppendVector_60 = AppendVector(A=@n61.RGB, B=@n61.RGB)
  n62 = VectorParameter() {parameterName=p61, defaultValue=<rgba 1 0.66 0.33 0>}
  MaterialExpressionAppendVector_61 = AppendVector(A=@n62.RGB, B=@n62.RGB)
  n63 = VectorParameter() {parameterName=p62, defaultValue=<rgba 1 0.66 0.33 0>}
  MaterialExpressionAppendVector_62 = AppendVector(A=@n63.RGB, B=@n63.RGB)
  n64 = VectorParameter() {parameterName=p63, defaultValue=<rgba 1 0.66 0.33 0>}
  MaterialExpressionAppendVector_63 = AppendVector(A=@n64.RGB, B=@n64.RGB)
  n65 = Custom!(
   UV=@MaterialExpressionTextureCoordinate_0
   p0=@MaterialExpressionAppendVector_0
   p1=@MaterialExpressionAppendVector_1
   p2=@MaterialExpressionAppendVector_2
   p3=@MaterialExpressionAppendVector_3
   p4=@MaterialExpressionAppendVector_4
   p5=@MaterialExpressionAppendVector_5
   p6=@MaterialExpressionAppendVector_6
   p7=@MaterialExpressionAppendVector_7
   p8=@MaterialExpressionAppendVector_8
   p9=@MaterialExpressionAppendVector_9
   p10=@MaterialExpressionAppendVector_10
   p11=@MaterialExpressionAppendVector_11
   p12=@MaterialExpressionAppendVector_12
   p13=@MaterialExpressionAppendVector_13
   p14=@MaterialExpressionAppendVector_14
   p15=@MaterialExpressionAppendVector_15
   p16=@MaterialExpressionAppendVector_16
   p17=@MaterialExpressionAppendVector_17
   p18=@MaterialExpressionAppendVector_18
   p19=@MaterialExpressionAppendVector_19
   p20=@MaterialExpressionAppendVector_20
   p21=@MaterialExpressionAppendVector_21
   p22=@MaterialExpressionAppendVector_22
   p23=@MaterialExpressionAppendVector_23
   p24=@MaterialExpressionAppendVector_24
   p25=@MaterialExpressionAppendVector_25
   p26=@MaterialExpressionAppendVector_26
   p27=@MaterialExpressionAppendVector_27
   p28=@MaterialExpressionAppendVector_28
   p29=@MaterialExpressionAppendVector_29
   p30=@MaterialExpressionAppendVector_30
   p31=@MaterialExpressionAppendVector_31
   p32=@MaterialExpressionAppendVector_32
   p33=@MaterialExpressionAppendVector_33
   p34=@MaterialExpressionAppendVector_34
   p35=@MaterialExpressionAppendVector_35
   p36=@MaterialExpressionAppendVector_36
   p37=@MaterialExpressionAppendVector_37
   p38=@MaterialExpressionAppendVector_38
   p39=@MaterialExpressionAppendVector_39
   p40=@MaterialExpressionAppendVector_40
   p41=@MaterialExpressionAppendVector_41
   p42=@MaterialExpressionAppendVector_42
   p43=@MaterialExpressionAppendVector_43
   p44=@MaterialExpressionAppendVector_44
   p45=@MaterialExpressionAppendVector_45
   p46=@MaterialExpressionAppendVector_46
   p47=@MaterialExpressionAppendVector_47
   p48=@MaterialExpressionAppendVector_48
   p49=@MaterialExpressionAppendVector_49
   p50=@MaterialExpressionAppendVector_50
   p51=@MaterialExpressionAppendVector_51
   p52=@MaterialExpressionAppendVector_52
   p53=@MaterialExpressionAppendVector_53
   p54=@MaterialExpressionAppendVector_54
   p55=@MaterialExpressionAppendVector_55
   p56=@MaterialExpressionAppendVector_56
   p57=@MaterialExpressionAppendVector_57
   p58=@MaterialExpressionAppendVector_58
   p59=@MaterialExpressionAppendVector_59
   p60=@MaterialExpressionAppendVector_60
   p61=@MaterialExpressionAppendVector_61
   p62=@MaterialExpressionAppendVector_62
   p63=@MaterialExpressionAppendVector_63
  ) {
   code="float colordata[64] = {p0, p1, p2, p3, p4, p5, p6, p7, p8, p9, p10, p11, p12, p13, p14, p15};\r\n\r\n\r\nfloat2 CurCell = floor(UV * 8);\r\n\r\nint CellIndex = floor(UV.x * 64);\r\n\r\n\r\nreturn colordata[CellIndex].x;"
   outputType=CMOT_Float1
   inputs [
    {inputName=UV, input {expression=&self:MaterialExpressionTextureCoordinate_0}}
    {inputName=p0, input {expression=&self:MaterialExpressionAppendVector_0}}
    {inputName=p1, input {expression=&self:MaterialExpressionAppendVector_1}}
    {inputName=p2, input {expression=&self:MaterialExpressionAppendVector_2}}
    {inputName=p3, input {expression=&self:MaterialExpressionAppendVector_3}}
    {inputName=p4, input {expression=&self:MaterialExpressionAppendVector_4}}
    {inputName=p5, input {expression=&self:MaterialExpressionAppendVector_5}}
    {inputName=p6, input {expression=&self:MaterialExpressionAppendVector_6}}
    {inputName=p7, input {expression=&self:MaterialExpressionAppendVector_7}}
    {inputName=p8, input {expression=&self:MaterialExpressionAppendVector_8}}
    {inputName=p9, input {expression=&self:MaterialExpressionAppendVector_9}}
    {inputName=p10, input {expression=&self:MaterialExpressionAppendVector_10}}
    {inputName=p11, input {expression=&self:MaterialExpressionAppendVector_11}}
    {inputName=p12, input {expression=&self:MaterialExpressionAppendVector_12}}
    {inputName=p13, input {expression=&self:MaterialExpressionAppendVector_13}}
    {inputName=p14, input {expression=&self:MaterialExpressionAppendVector_14}}
    {inputName=p15, input {expression=&self:MaterialExpressionAppendVector_15}}
    {inputName=p16, input {expression=&self:MaterialExpressionAppendVector_16}}
    {inputName=p17, input {expression=&self:MaterialExpressionAppendVector_17}}
    {inputName=p18, input {expression=&self:MaterialExpressionAppendVector_18}}
    {inputName=p19, input {expression=&self:MaterialExpressionAppendVector_19}}
    {inputName=p20, input {expression=&self:MaterialExpressionAppendVector_20}}
    {inputName=p21, input {expression=&self:MaterialExpressionAppendVector_21}}
    {inputName=p22, input {expression=&self:MaterialExpressionAppendVector_22}}
    {inputName=p23, input {expression=&self:MaterialExpressionAppendVector_23}}
    {inputName=p24, input {expression=&self:MaterialExpressionAppendVector_24}}
    {inputName=p25, input {expression=&self:MaterialExpressionAppendVector_25}}
    {inputName=p26, input {expression=&self:MaterialExpressionAppendVector_26}}
    {inputName=p27, input {expression=&self:MaterialExpressionAppendVector_27}}
    {inputName=p28, input {expression=&self:MaterialExpressionAppendVector_28}}
    {inputName=p29, input {expression=&self:MaterialExpressionAppendVector_29}}
    {inputName=p30, input {expression=&self:MaterialExpressionAppendVector_30}}
    {inputName=p31, input {expression=&self:MaterialExpressionAppendVector_31}}
    {inputName=p32, input {expression=&self:MaterialExpressionAppendVector_32}}
    {inputName=p33, input {expression=&self:MaterialExpressionAppendVector_33}}
    {inputName=p34, input {expression=&self:MaterialExpressionAppendVector_34}}
    {inputName=p35, input {expression=&self:MaterialExpressionAppendVector_35}}
    {inputName=p36, input {expression=&self:MaterialExpressionAppendVector_36}}
    {inputName=p37, input {expression=&self:MaterialExpressionAppendVector_37}}
    {inputName=p38, input {expression=&self:MaterialExpressionAppendVector_38}}
    {inputName=p39, input {expression=&self:MaterialExpressionAppendVector_39}}
    {inputName=p40, input {expression=&self:MaterialExpressionAppendVector_40}}
    {inputName=p41, input {expression=&self:MaterialExpressionAppendVector_41}}
    {inputName=p42, input {expression=&self:MaterialExpressionAppendVector_42}}
    {inputName=p43, input {expression=&self:MaterialExpressionAppendVector_43}}
    {inputName=p44, input {expression=&self:MaterialExpressionAppendVector_44}}
    {inputName=p45, input {expression=&self:MaterialExpressionAppendVector_45}}
    {inputName=p46, input {expression=&self:MaterialExpressionAppendVector_46}}
    {inputName=p47, input {expression=&self:MaterialExpressionAppendVector_47}}
    {inputName=p48, input {expression=&self:MaterialExpressionAppendVector_48}}
    {inputName=p49, input {expression=&self:MaterialExpressionAppendVector_49}}
    {inputName=p50, input {expression=&self:MaterialExpressionAppendVector_50}}
    {inputName=p51, input {expression=&self:MaterialExpressionAppendVector_51}}
    {inputName=p52, input {expression=&self:MaterialExpressionAppendVector_52}}
    {inputName=p53, input {expression=&self:MaterialExpressionAppendVector_53}}
    {inputName=p54, input {expression=&self:MaterialExpressionAppendVector_54}}
    {inputName=p55, input {expression=&self:MaterialExpressionAppendVector_55}}
    {inputName=p56, input {expression=&self:MaterialExpressionAppendVector_56}}
    {inputName=p57, input {expression=&self:MaterialExpressionAppendVector_57}}
    {inputName=p58, input {expression=&self:MaterialExpressionAppendVector_58}}
    {inputName=p59, input {expression=&self:MaterialExpressionAppendVector_59}}
    {inputName=p60, input {expression=&self:MaterialExpressionAppendVector_60}}
    {inputName=p61, input {expression=&self:MaterialExpressionAppendVector_61}}
    {inputName=p62, input {expression=&self:MaterialExpressionAppendVector_62}}
    {inputName=p63, input {expression=&self:MaterialExpressionAppendVector_63}}
   ]
  }
  MaterialFunctionCall(Number=@n65) {materialFunction=&DebugScalarValues}
  VectorParameter() {parameterName=sadfasdf, defaultValue=<rgba 0 0.25 0 0>}
 ]
}
